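; Looping gameplay cue that shows Fire Petal shrine activation progress, themed by the sprite type in the cue's aggregated source tags.
; On application, it sets the RingEmissive color on the shrine's mesh and base, stamping game time into custom primitive data.
; On loop start, it swaps the first particle component's Niagara system to the matching variant and plays a sound at location, with a dedicated cue for ZeroPoint.
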(event EventOnApplication(EitherVFX-preferthisone)
  (params (MyTarget :type Actor%20Object%20Reference) (Parameters :type Gameplay%20Cue%20Parameters%20Structure%20%28by%20ref%29) (ParticleComponents :type Array%20of%20FXSystem%20Component%20Object%20References) (AudioComponents :type Array%20of%20Audio%20Component%20Object%20References) (BurstCameraShakeInstance :type Camera%20Shake%20Base%20Object%20Reference) (BurstDecalInstance :type Decal%20Actor%20Object%20Reference))
  (bind (_normalizedmagnitude _rawmagnitude _effectcontext _matchedtagname _originaltag _aggregatedsourcetags) (Ability|GameplayCue|BreakGameplayCueParameters Parameters))
  (bind _asbga_fire_petal_shrine_parent (Utilities|Casting|CastToBGA_FirePetal_Shrine_Parent MyTarget))
  (bind _emissivecolor (CallFunction|GetActivationEmissiveColortoUse _aggregatedsourcetags))
  (Rendering|Material|SetVectorParameterValueonMaterials (Variables|/FirePetalShrineGameplay/Gameplay/Actors/BGA_FirePetal_Shrine_Parent.BGA_FirePetal_Shrine_Parent_C|GetMesh _asbga_fire_petal_shrine_parent) "RingEmissive" _emissivecolor)
  (Rendering|Material|SetVectorParameterValueonMaterials (Variables|/FirePetalShrineGameplay/Gameplay/Actors/BGA_FirePetal_Shrine_Parent.BGA_FirePetal_Shrine_Parent_C|GetBase _asbga_fire_petal_shrine_parent) "RingEmissive" _emissivecolor)
  (Rendering|Material|SetCustomPrimitiveDataFloat (Variables|/FirePetalShrineGameplay/Gameplay/Actors/BGA_FirePetal_Shrine_Parent.BGA_FirePetal_Shrine_Parent_C|GetBase _asbga_fire_petal_shrine_parent) 0 (Utilities|Time|GetGameTimeinSeconds _asbga_fire_petal_shrine_parent))
  (Rendering|Material|SetCustomPrimitiveDataFloat (Variables|/FirePetalShrineGameplay/Gameplay/Actors/BGA_FirePetal_Shrine_Parent.BGA_FirePetal_Shrine_Parent_C|GetMesh _asbga_fire_petal_shrine_parent) 0 (Utilities|Time|GetGameTimeinSeconds _asbga_fire_petal_shrine_parent)))

(event EventOnLoopingStart(EitherVFX-preferthisversion)
  (params (MyTarget :type Actor%20Object%20Reference) (Parameters :type Gameplay%20Cue%20Parameters%20Structure%20%28by%20ref%29) (ParticleComponents :type Array%20of%20FXSystem%20Component%20Object%20References) (AudioComponents :type Array%20of%20Audio%20Component%20Object%20References))
  (bind (_normalizedmagnitude _rawmagnitude _effectcontext _matchedtagname _originaltag _aggregatedsourcetags) (Ability|GameplayCue|BreakGameplayCueParameters Parameters))
  (bind _asniagara_particle_system_component (Utilities|Casting|CastToNiagaraComponent (Utilities|Array|Get(acopy) ParticleComponents)))
  (bind _niagarasystem (CallFunction|GetActivationLoopNiagaratoUse _aggregatedsourcetags))
  (Niagara|SetNiagaraSystemAsset _asniagara_particle_system_component _niagarasystem true)
  (bind _soundcue (CallFunction|GetSoundCuetoUse _aggregatedsourcetags))
  (Audio|PlaySoundatLocation MyTarget _soundcue (Transformation|GetActorLocation)))

(fn GetActivationEmissiveColorToUse
  (params (Selection :type Gameplay%20Tag%20Container%20Structure))
  (switch Utilities|FlowControl|Switch|SwitchonGameplayTagContainer Selection
    (:(GameplayTags=((TagName="Item.Sprite.Air")))
      (return (Variables|Default|GetAirSpriteActivationColor)))
    (:(GameplayTags=((TagName="Item.Sprite.Water")))
      (return (Variables|Default|GetWaterSpriteActivationColor)))
    (:(GameplayTags=((TagName="Item.Sprite.ZeroPoint")))
      (return (Variables|Default|GetSuperSpriteActivationColor)))
    (:(GameplayTags=((TagName="Item.Sprite.WhisperRemedySweet")))
      (return (Variables|Default|GetWhisperRemedyActivationColor)))))

(fn GetActivationLoopNiagaraToUse
  (params (Selection :type Gameplay%20Tag%20Container%20Structure))
  (switch Utilities|FlowControl|Switch|SwitchonGameplayTagContainer Selection
    (:(GameplayTags=((TagName="Item.Sprite.Air")))
      (return "/FirePetalShrineGameplay/Assets/FX/Air/NS_FirePetal_Shrine_ActivationLoop_Air.NS_FirePetal_Shrine_ActivationLoop_Air"))
    (:(GameplayTags=((TagName="Item.Sprite.Water")))
      (return "/FirePetalShrineGameplay/Assets/FX/Water/NS_FirePetal_Shrine_ActivationLoop_Water.NS_FirePetal_Shrine_ActivationLoop_Water"))
    (:(GameplayTags=((TagName="Item.Sprite.ZeroPoint")))
      (return "/FirePetalShrineGameplay/Assets/FX/SuperSprite/NS_FirePetal_Shrine_ActivationLoop_SuperSprite.NS_FirePetal_Shrine_ActivationLoop_SuperSprite"))
    (:(GameplayTags=((TagName="Item.Sprite.WhisperRemedySweet")))
      (return "/FirePetalShrineGameplay/Assets/FX/WhisperRemedySweet/NS_FirePetal_Shrine_ActivationLoop_WhisperRemedySweet.NS_FirePetal_Shrine_ActivationLoop_WhisperRemedySweet"))))

(fn GetSoundCueToUse
  (params (Selection :type Gameplay%20Tag%20Container%20Structure))
  (switch Utilities|FlowControl|Switch|SwitchonGameplayTagContainer Selection
    (:Default
      (return "/FirePetalShrineGameplay/Assets/Sounds/Cues/SC_FirePetalShrine_Sprite_Activate_Progress.SC_FirePetalShrine_Sprite_Activate_Progress"))
    (:(GameplayTags=((TagName="Item.Sprite.ZeroPoint")))
      (return "/FirePetalShrineGameplay/Assets/Sounds/Cues/SC_SideKick_Sprite_Shrine_Loop_Cue.SC_SideKick_Sprite_Shrine_Loop_Cue"))))

(fn ConstructionScript)
